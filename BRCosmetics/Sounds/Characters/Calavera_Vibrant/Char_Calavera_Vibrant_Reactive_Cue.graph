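# Sound cue for the reactive sound of the Calavera Vibrant character cosmetic.
# A Random node picks one of three React wave variations with equal weight, and a fixed short Delay at the root holds back playback.
# It plays through the player foley sound class with basic foley attenuation and a boosted volume multiplier.
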SoundCue /BRCosmetics/Sounds/Characters/Calavera_Vibrant/Char_Calavera_Vibrant_Reactive_Cue
root=Delay_0
nodes [
 {id=Delay_0, children [Random_1], properties {delayMax=0.2, delayMin=0.2}}
 {id=Random_1, children [WavePlayer_0,WavePlayer_1,WavePlayer_2], properties {weights [1,1,1]}}
 {id=WavePlayer_0, properties {soundWaveAssetPtr=&/Char_Calavera_Vibrant_React_03}}
 {id=WavePlayer_1, properties {soundWaveAssetPtr=&/Char_Calavera_Vibrant_React_01}}
 {id=WavePlayer_2, properties {soundWaveAssetPtr=&/Char_Calavera_Vibrant_React_02}}
]
properties {
 attenuationSettings=&Foley_Attenuation_Basic_Att
 maxDistance=1536
 soundClassObject=&Fort_Player_Foley
 volumeMultiplier=3
}
